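# Translucent UI material for the Force Powers reticle that draws the cooldown state of the Push power.
# Six-vertex polygon SDFs build an idle shape plus middle and mirrored side parts, the latter two masked by gradients with separate start parameters.
# AnimateCooldown blends between idle and active looks, with FinalOpacity scaling the result.
# The active color gets a scrolling highlight from a panned sine raised to a high power.
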Material /PlotTwist/ForcePowers/Shared/Reticle/Materials/M_UI_CoolDownStatePush
graph {
 properties {blendMode=BLEND_Translucent, materialDomain=MD_UI, pixelDepthOffsetMode=PDOM_Legacy, usedShadingModels=Default Lit}
 nodes [
  n1 = NamedRerouteUsage()
  n2 = VectorParameter() {parameterName=Color, defaultValue=<rgba 0.0578 0.5271 0 1>}
  n3 = MaterialFunctionCall(Texture Scale=Constant() {r=1.5}) {materialFunction=&ScaleUVsByCenter}
  n4 = NamedRerouteUsage()
  n5 = VectorParameter() {parameterName=Color, defaultValue=<rgba 0.0578 0.5271 0 1>}
  Add(
   A=Multiply(
    A=Multiply(A=@n1, B=@n2.RGB)
    B=MaterialFunctionCall(UV=@n3.UVs) {materialFunction=&MF_UI_RadialGradient}.Smooth
   )
   B=Multiply(
    A=Multiply(A=@n4, B=@n5.RGB)
    B=MaterialFunctionCall(UV=@n3.UVs) {materialFunction=&MF_UI_RadialGradient}.Smooth
   )
  )
  LinearInterpolate(A=NamedRerouteUsage(), B=NamedRerouteUsage(), Alpha=ScalarParameter() {parameterName=AnimateCooldown})
  MaterialFunctionCall(GradientStart=ScalarParameter() {parameterName=GradientStart}, GradientEnd=Constant(), RotationAngle=Constant() {r=0.25}) {materialFunction=&MF_UI_LinearGradient}
  n6 = Frac(None=Time())
  MaterialFunctionCall(GradientStart=Sine(None=OneMinus(None=@n6)), RotationAngle=Constant() {r=0.25}) {materialFunction=&MF_UI_LinearGradient}
  n7 = NamedRerouteUsage()
  n8 = ScalarParameter() {defaultValue=0.6, parameterName=IdlePartOpacity}
  n9 = NamedRerouteUsage()
  n10 = OneMinus(None=MaterialFunctionCall() {materialFunction=&LinearGradient}.VGradient)
  n11 = Multiply(A=@n10, B=Constant() {r=2})
  Multiply(
   A=LinearInterpolate(A=Multiply(A=@n7, B=@n8), B=Multiply(A=@n9, B=@n11), Alpha=ScalarParameter() {parameterName=AnimateCooldown})
   B=ScalarParameter() {defaultValue=1, parameterName=FinalOpacity}
  )
  n12 = MaterialFunctionCall(Texture Scale=Constant() {r=0.7}) {materialFunction=&ScaleUVsByCenter}
  n13 = Constant2Vector() {g=-0.1783, r=-0.9516}
  n14 = Constant2Vector() {g=-0.416, r=-0.7056}
  n15 = Constant2Vector() {g=-0.6811}
  n16 = Constant2Vector() {g=-0.4251, r=0.7131}
  n17 = Constant2Vector() {g=-0.1737, r=0.9771}
  n18 = Constant2Vector() {g=-0.3383}
  n19 = OneMinus(None=SmoothStep(Value=MaterialFunctionCall!(
     UVs=@n12.UVs
     vertex_0=@n13
     vertex_1=@n14
     vertex_2=@n15
     vertex_3=@n16
     vertex_4=@n17
     vertex_5=@n18
     Glow Min
     Glow Max
     Outline Thickness
    ) {materialFunction=/PlotTwist/ForcePowers/Shared/Reticle/Materials/MaterialFuctions/MF_6VertexPolygon}.SDF) {constMax=0.1})
  NamedRerouteDeclaration(None=@n19) {name=IdlePart, nodeColor=<rgba 0 0 0 1>}
  NamedRerouteDeclaration(None=@n19) {name=IdlePartOpacity, nodeColor=<rgba 0 0 0 1>}
  n20 = NamedRerouteUsage()
  n21 = MaterialFunctionCall(
   GradientStart=ScalarParameter() {defaultValue=0.001, parameterName=GradientStartMiddle}
   GradientEnd=Constant()
   RotationAngle=Constant() {r=0.25}
  ) {materialFunction=&MF_UI_LinearGradient}
  n22 = NamedRerouteUsage()
  n23 = MaterialFunctionCall(
   GradientStart=ScalarParameter() {defaultValue=0.001, parameterName=GradientStartSides}
   GradientEnd=Constant()
   RotationAngle=Constant() {r=0.25}
  ) {materialFunction=&MF_UI_LinearGradient}
  NamedRerouteDeclaration(None=Add(A=Multiply(A=@n20, B=@n21.Result), B=Multiply(A=@n22, B=@n23.Result))) {name=OpacityActive, nodeColor=<rgba 0 0 0 1>}
  n24 = Constant() {r=0.65}
  n25 = Constant() {r=0.7}
  n26 = MaterialFunctionCall(
   UVs=MaterialFunctionCall!(UV, Offset=Constant2Vector() {g=-0.06, r=0.11}, Rotation=Constant() {r=-0.03}, Scale=AppendVector(A=@n24, B=@n25), MirrorY) {materialFunction=/FortUILibrary/MaterialFunctions/Basics/MF_UI_Symmetry_2x1_DEPRECATED}.Result
   vertex_0=Constant2Vector() {g=0.9097, r=-0.9359}
   vertex_1=Constant2Vector() {g=0.512, r=0.2727}
   vertex_2=Constant2Vector() {g=-0.9509, r=0.9234}
   vertex_3=Constant2Vector() {g=-0.8481, r=0.6643}
   vertex_4=Constant2Vector() {g=-0.3383, r=0.1966}
   vertex_5=Constant2Vector() {g=-0.3383, r=0.1966}
  ) {materialFunction=&MF_6VertexPolygon}
  NamedRerouteDeclaration(None=OneMinus(None=SmoothStep(Value=@n26.SDF) {constMax=0.1})) {name=SideParts, nodeColor=<rgba 0 0 0 1>}
  n27 = MaterialFunctionCall(
   UVs=MaterialFunctionCall(UVs=TextureCoordinate(), Texture Scale=Constant() {r=0.5}) {materialFunction=&ScaleUVsByCenter}.UVs
   vertex_0=Constant2Vector() {g=-0.8241, r=-0.3118}
   vertex_1=Constant2Vector() {g=-0.9646}
   vertex_2=Constant2Vector() {g=-0.8241, r=0.3118}
   vertex_3=Constant2Vector() {g=0.9667, r=0.9157}
   vertex_4=Constant2Vector() {g=0.6857}
   vertex_5=Constant2Vector() {g=0.9667, r=-0.9157}
  ) {materialFunction=&MF_6VertexPolygon}
  NamedRerouteDeclaration(None=OneMinus(None=SmoothStep(Value=@n27.SDF) {constMax=0.1})) {name=MiddlePart, nodeColor=<rgba 0 0 0 1>}
  n28 = Panner(Coordinate=TextureCoordinate()) {speedY=1.75}
  n29 = Power(Base=Sine(None=ComponentMask(None=@n28) {g=true})) {constExponent=20}
  NamedRerouteDeclaration(None=Multiply(A=@n5.RGB, B=Add(A=@n29) {constB=0.6})) {name=ActiveFinalColor, nodeColor=<rgba 0 0 0 1>}
  SmoothStep(Value=@n10) {constMax=0.6, constMin=-0.1}
 ]
}
