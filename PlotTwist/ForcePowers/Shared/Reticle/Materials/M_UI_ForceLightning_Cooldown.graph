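# Translucent UI material for the Force Lightning reticle cooldown icon, drawing a lightning bolt that shows recharge progress.
# Channel-packed distance field functions turn a lightning SDF texture into background and fill shapes.
# Animated start and end progress parameters drive a soft-edged wipe across slightly tilted, rotated UVs.
# An isReady parameter blends the lit color from white to a red radial gradient over a near-black red base.
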Material /PlotTwist/ForcePowers/Shared/Reticle/Materials/M_UI_ForceLightning_Cooldown
graph {
 properties {bAutomaticallySetUsageInEditor=false, blendMode=BLEND_Translucent, materialDomain=MD_UI, usedShadingModels=Default Lit}
 nodes [
  n1 = Multiply(A=NamedRerouteUsage(), B=NamedRerouteUsage())
  Add(A=Multiply(A=NamedRerouteUsage()) {constB=0.4}, B=@n1)
  Constant() {r=0.5}
  n2 = Multiply(A=Constant3Vector() {constant=<rgba 1 0 0 1>}) {constB=2}
  n3 = Constant3Vector() {constant=<rgba 0.6 0 0 1>}
  n4 = MaterialFunctionCall() {materialFunction=&RadialGradientExponential}
  LinearInterpolate(
   A=Constant3Vector() {constant=<rgba 0.01 0 0 0>}
   B=LinearInterpolate(B=LinearInterpolate(A=@n2, B=@n3, Alpha=@n4.RadialGradientExponential), Alpha=ScalarParameter() {defaultValue=1, parameterName=isReady}) {constA=1}
   Alpha=@n1
  )
  Multiply(
   A=@n3
   B=MaterialFunctionCall!(UVs, Offset, RotationCenter, RotationAngle=Constant() {r=0.3}) {materialFunction=/FortUILibrary/MaterialFunctions/Gradients/MF_UI_ReflectedGradient}.Result
  )
  NamedRerouteDeclaration(None=MaterialFunctionCall!(
    RGBA Channels=TextureSample() {texture=/PlotTwist/ForcePowers/Shared/Reticle/Textures/T_UI_Lightning_SDF}.R
    Glow Min=Constant() {r=0.2}
    Glow Max=Constant() {r=0.4}
    Outline Thickness
   ) {materialFunction=/FortUILibrary/MaterialFunctions/Basics/MF_UI_ChannelPackedDistanceField}.Fill) {name=LightningShapeBackground, nodeColor=<rgba 0 0 0 1>}
  NamedRerouteDeclaration(None=MaterialFunctionCall(
    RGBA Channels=TextureSample() {texture=&T_UI_Lightning_SDF}.R
    Glow Min=Constant() {r=0.5}
    Glow Max=Constant() {r=0.8}
   ) {materialFunction=&MF_UI_ChannelPackedDistanceField}.Fill) {name=LightningIconFill, nodeColor=<rgba 0 0 0 1>}
  n5 = ScalarParameter() {defaultValue=0.03, group=Progress_Layout, parameterName=ProgressEdgeSharpness, sliderMax=1, sortPriority=2}
  n6 = MaterialFunctionCall(UVs=TextureCoordinate(), "Rotation Angle (0-1)"=Constant() {r=0.265}) {materialFunction=&CustomRotator}
  n7 = ScalarParameter() {
   defaultValue=0.2
   desc="When progress value changes, it defines where change starts.\r\n0 - 1"
   group=Progress
   parameterName=PipProgressStart
   sliderMax=1
   sortPriority=1
  }
  n8 = ScalarParameter() {
   defaultValue=0.88
   desc="When progress value changes, it defines where change ends.\r\n0 - 1"
   group=Progress
   parameterName=PipProgressEnd
   sliderMax=1
   sortPriority=2
  }
  n9 = ScalarParameter() {
   defaultValue=1
   desc="Property for animating progress value change (from ProgressStart to ProgressEnd)\r\n0 - 1"
   group=Progress
   parameterName=AnimatePipProgress
   sliderMax=1
   sortPriority=3
  }
  n10 = LinearInterpolate(A=@n7, B=@n8, Alpha=@n9)
  n11 = ComponentMask(None=Add(A=Multiply(A=@n6.Rotated Values) {constB=-1}, B=NamedRerouteDeclaration(None=@n10) {name=Progress_1, nodeColor=<rgba 0 0 0 1>}) {constB=0.5}) {r=true}
  NamedRerouteDeclaration(None=Saturate(None=SmoothStep(Max=@n5, Value=@n11) {constMax=0.01})) {name=ProgressGraident_Vertical, nodeColor=<rgba 0 0 0 1>}
 ]
}
