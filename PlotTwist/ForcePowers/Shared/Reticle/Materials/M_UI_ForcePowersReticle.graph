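# Translucent UI material that draws the Force Powers reticle from a middle rounded-box outline, a bottom shape and a top icon with rounded box.
# Textures and MF_UI_Box SDFs form the masks, with isForceLightning blending the bottom shape toward a distance-field bracket.
# Scalar parameters drive top-part position and width, cooldown, a time-based sine flash and opacity.
# Top-part color blends from a default toward a red, blue or green selection.
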Material /PlotTwist/ForcePowers/Shared/Reticle/Materials/M_UI_ForcePowersReticle
graph {
 properties {blendMode=BLEND_Translucent, materialDomain=MD_UI, pixelDepthOffsetMode=PDOM_Legacy, usedShadingModels=Default Lit}
 nodes [
  n1 = Max(A=Max(A=NamedRerouteUsage(), B=NamedRerouteUsage()), B=NamedRerouteUsage())
  Add(A=Multiply(A=NamedRerouteUsage(), B=NamedRerouteUsage()), B=@n1)
  n2 = NamedRerouteUsage()
  n3 = Frac(None=Time())
  n4 = LinearInterpolate(A=NamedRerouteUsage(), B=Sine(None=Multiply(A=@n3) {constB=0.5}), Alpha=NamedRerouteUsage())
  Multiply(A=Max(A=@n1, B=Multiply(A=@n2, B=@n4)), B=NamedRerouteUsage())
  NamedRerouteDeclaration(None=ScalarParameter() {defaultValue=0.8, group=Scale, parameterName=IconScale}) {name=Icon Scale, nodeColor=<rgba 0 0 0 1>}
  NamedRerouteDeclaration(None=ScalarParameter() {defaultValue=-0.04, group=Position, parameterName=TopPart Y PositionBottom}) {name=TopPart Y PositionBottom, nodeColor=<rgba 0 0 0 1>}
  NamedRerouteDeclaration(None=ScalarParameter() {defaultValue=0.08, group=Position, parameterName=TopPart Y PositionReady1}) {name=TopPart Y PositionReady1, nodeColor=<rgba 0 0 0 1>}
  n5 = NamedRerouteDeclaration(None=TextureCoordinate()) {name=CoordUV, nodeColor=<rgba 0 0 0 1>}
  NamedRerouteDeclaration(None=ComponentMask(None=@n5) {r=true}) {name=CoordU, nodeColor=<rgba 0 0 0 1>}
  NamedRerouteDeclaration(None=ComponentMask(None=@n5) {g=true}) {name=CoordV, nodeColor=<rgba 0 0 0 1>}
  NamedRerouteDeclaration(None=ScalarParameter() {defaultValue=1, desc="0-Red, 0.5-Blue, 1-Green", group=Shape, parameterName=ReticleBottomPart, sliderMax=1}) {name=ReticleBottomPart, nodeColor=<rgba 0 0 0 1>}
  NamedRerouteDeclaration(None=ScalarParameter() {group=Animation, parameterName=AnimateTopPart, sliderMax=1}) {name=AnimateTopPart, nodeColor=<rgba 0 0 0 1>}
  NamedRerouteDeclaration(None=ScalarParameter() {defaultValue=1, group=Opacity, parameterName=TopPartOpacity, sliderMax=1}) {name=TopPartOpacity, nodeColor=<rgba 0 0 0 1>}
  NamedRerouteDeclaration(None=ScalarParameter() {group=Animation, parameterName=TopPartFlash, sliderMax=1}) {name=TopPartFlash, nodeColor=<rgba 0 0 0 1>}
  NamedRerouteDeclaration(None=ScalarParameter() {defaultValue=1, group=Opacity, parameterName=FinalOpacity, sliderMax=1}) {name=FinalOpacity, nodeColor=<rgba 0 0 0 1>}
  NamedRerouteDeclaration(None=VectorParameter() {group=Color, parameterName=TopPartRed, defaultValue=<rgba 0.9734 0.0006 0.0006 1>}.RGB) {name=TopPartRed, nodeColor=<rgba 0 0 0 1>}
  NamedRerouteDeclaration(None=ScalarParameter() {defaultValue=-0.08, group=Position, parameterName=TopPart Y Bottom1}) {name=TopPart Y Bottom1, nodeColor=<rgba 0 0 0 1>}
  NamedRerouteDeclaration(None=VectorParameter() {group=Color, parameterName=TopPartBlue, defaultValue=<rgba 0.0086 0.3916 0.5841 1>}.RGB) {name=TopPartBlue, nodeColor=<rgba 0 0 0 1>}
  NamedRerouteDeclaration(None=VectorParameter() {group=Color, parameterName=TopPartGreen, defaultValue=<rgba 0.0423 0.552 0.006 1>}.RGB) {name=TopPartGreen, nodeColor=<rgba 0 0 0 1>}
  NamedRerouteDeclaration(None=VectorParameter() {group=Color, parameterName=DefaultColor, defaultValue=<rgba 0.5923 0.5923 0.5923 1>}.RGB) {name=DefaultColor, nodeColor=<rgba 0 0 0 1>}
  NamedRerouteDeclaration(None=ScalarParameter() {defaultValue=0.4, group=Scale, parameterName=TopPartScaleX}) {name=TopPartScaleX, nodeColor=<rgba 0 0 0 1>}
  NamedRerouteDeclaration(None=ScalarParameter() {defaultValue=0.45, group=Scale, parameterName=TopPartScaleXScaled}) {name=TopPartSacaleXScaled, nodeColor=<rgba 0 0 0 1>}
  NamedRerouteDeclaration(None=ScalarParameter() {defaultValue=0.34, group=Position, parameterName=TopPart Y BoxPosition}) {name=TopPart Y BoxPosition, nodeColor=<rgba 0 0 0 1>}
  NamedRerouteDeclaration(None=ScalarParameter() {defaultValue=0.29, group=Position, parameterName=TopPart Y BoxPositionReady}) {name=TopPart Y BoxPositionReady, nodeColor=<rgba 0 0 0 1>}
  n6 = NamedRerouteUsage()
  n7 = LinearInterpolate(A=NamedRerouteUsage(), B=NamedRerouteUsage(), Alpha=NamedRerouteUsage())
  n8 = LinearInterpolate(A=NamedRerouteUsage(), B=NamedRerouteUsage(), Alpha=NamedRerouteUsage())
  n9 = Step(X=NamedRerouteUsage()) {constY=0.5}
  n10 = NamedRerouteDeclaration(None=Add(A=NamedRerouteUsage(), B=LinearInterpolate(A=@n7, B=@n8, Alpha=@n9)) {constB=0.08}) {name=IconPosition, nodeColor=<rgba 0 0 0 1>}
  n11 = TextureSample(UVs=MaterialFunctionCall(UVs=AppendVector(A=@n6, B=@n10), Texture Scale=NamedRerouteUsage()) {materialFunction=&ScaleUVsByCenter}.UVs) {samplerSource=SSM_Clamp_WorldGroupSettings, texture=/PlotTwist/ForcePowers/Shared/Reticle/Textures/T_UI_Reticle_ForcePowers}
  NamedRerouteDeclaration(None=@n11.A) {name=TopPart, nodeColor=<rgba 0 0 0 1>}
  n12 = NamedRerouteUsage()
  n13 = Add(A=LinearInterpolate(A=NamedRerouteUsage(), B=NamedRerouteUsage(), Alpha=NamedRerouteUsage()), B=NamedRerouteUsage())
  n14 = NamedRerouteUsage()
  n15 = NamedRerouteUsage()
  n16 = NamedRerouteUsage()
  n17 = MaterialFunctionCall(
   UVs=MaterialFunctionCall(UVs=AppendVector(A=@n12, B=@n13), Texture Scale=Constant() {r=0.35}) {materialFunction=&ScaleUVsByCenter}.UVs
   Size=AppendVector(A=LinearInterpolate(A=@n14, B=@n15, Alpha=@n16), B=Constant() {r=0.1})
   CornerRoundness=Constant() {r=0.1}
   OutlineThickness=Constant() {r=0.02}
  ) {materialFunction=&MF_UI_Box}
  n18 = SmoothStep(Value=@n17.SDF) {constMax=0.05}
  n19 = SmoothStep(Value=@n17.SDFOutline) {constMax=0.01}
  NamedRerouteDeclaration(None=Add(A=OneMinus(None=@n18), B=OneMinus(None=@n19))) {name=RoundedBoxTop, nodeColor=<rgba 0 0 0 1>}
  n20 = NamedRerouteUsage()
  n21 = NamedRerouteUsage()
  n22 = NamedRerouteUsage()
  n23 = NamedRerouteUsage()
  NamedRerouteDeclaration(None=LinearInterpolate(
    A=NamedRerouteUsage()
    B=MaterialFunctionCall(A=@n20, B=@n21, C=@n22, Alpha=@n23) {materialFunction=&3ColorBlend}.Result
    Alpha=NamedRerouteUsage()
   )) {name=TopPartColorSelect, nodeColor=<rgba 0 0 0 1>}
  n24 = MaterialFunctionCall(
   UVs=MaterialFunctionCall(UVs=NamedRerouteUsage(), Texture Scale=Constant() {r=0.15}) {materialFunction=&ScaleUVsByCenter}.UVs
   Size=Constant2Vector() {g=0.4, r=0.8}
   CornerRoundness=Constant() {r=0.4}
   OutlineThickness=Constant() {r=0.1}
  ) {materialFunction=&MF_UI_Box}
  NamedRerouteDeclaration(None=OneMinus(None=SmoothStep(Value=@n24.SDFOutline) {constMax=0.03})) {name=MiddlePart, nodeColor=<rgba 0 0 0 1>}
  n25 = NamedRerouteUsage()
  n26 = TextureCoordinate()
  n27 = ComponentMask(None=@n26) {r=true}
  n28 = Add(A=ComponentMask(None=@n26) {g=true}, B=Constant() {r=-0.18})
  n29 = TextureSample(UVs=MaterialFunctionCall(UVs=AppendVector(A=@n27, B=@n28), Texture Scale=Constant() {r=0.9}) {materialFunction=&ScaleUVsByCenter}.UVs) {texture=/PlotTwist/ForcePowers/Shared/Reticle/Textures/T_UI_Bracket}
  n30 = Constant() {r=0.4}
  n31 = Constant() {r=0.6}
  NamedRerouteDeclaration(None=LinearInterpolate(
    A=MaterialFunctionCall(A=@n11.R, B=@n11.R, C=@n11.R, Alpha=@n25) {materialFunction=&3ColorBlend}.Result
    B=MaterialFunctionCall!(RGBA Channels=@n29.R, Glow Min=@n30, Glow Max=@n31, Outline Thickness) {materialFunction=/FortUILibrary/MaterialFunctions/Basics/MF_UI_ChannelPackedDistanceField}.Fill
    Alpha=ScalarParameter() {parameterName=isForceLightning}
   )) {name=Bottom Part, nodeColor=<rgba 0 0 0 1>}
  NamedRerouteDeclaration(None=ScalarParameter() {defaultValue=1, group=Animation, parameterName=AnimCooldown, sliderMax=1}) {name=AnimCooldown, nodeColor=<rgba 0 0 0 1>}
  NamedRerouteDeclaration(None=ScalarParameter() {defaultValue=0.04, group=Position, parameterName=TopPart Y Position1}) {name=TopPart Y Position1, nodeColor=<rgba 0 0 0 1>}
 ]
}
