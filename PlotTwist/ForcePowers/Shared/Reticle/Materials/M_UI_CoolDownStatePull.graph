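# Translucent UI material for the Force power reticle that draws arrow shapes for the pull power's idle and cooldown states.
# Two six-vertex polygon SDFs, offset, rotated, scaled and smoothstepped, form the arrow masks.
# AnimateCooldown shifts the first arrow from white to Color1 and blends idle opacity into separate per-arrow opacities, scaled by FinalOpacity.
# AnimateLines fades in a vertically panning, sharpened sine band that modulates brightness.
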Material /PlotTwist/ForcePowers/Shared/Reticle/Materials/M_UI_CoolDownStatePull
graph {
 properties {blendMode=BLEND_Translucent, materialDomain=MD_UI, pixelDepthOffsetMode=PDOM_Legacy, usedShadingModels=Default Lit}
 nodes [
  n1 = TextureCoordinate()
  n2 = AppendVector(
   A=ComponentMask(None=@n1) {r=true}
   B=Add(A=ComponentMask(None=@n1) {g=true}, B=ScalarParameter() {defaultValue=-0.1, parameterName=FirstArrowPosition}) {constB=-0.1}
  )
  n3 = Constant() {r=0.8771}
  n4 = MaterialFunctionCall!(
   UVs=MaterialFunctionCall(
    UVs=MaterialFunctionCall(UVs=@n2, "Rotation Angle (0-1)"=@n3) {materialFunction=&CustomRotator}.Rotated Values
    Texture Scale=Constant() {r=0.25}
   ) {materialFunction=&ScaleUVsByCenter}.UVs
   vertex_0=Constant2Vector() {g=-0.448, r=0.3131}
   vertex_1=Constant2Vector() {g=0.064, r=0.0183}
   vertex_2=Constant2Vector() {g=0.3032, r=-0.448}
   vertex_3=Constant2Vector() {g=0.9232, r=-0.9215}
   vertex_4=Constant2Vector() {g=0.3337, r=0.2926}
   vertex_5=Constant2Vector() {g=-0.8888, r=0.9097}
   Glow Min
   Glow Max
   Outline Thickness
  ) {materialFunction=/PlotTwist/ForcePowers/Shared/Reticle/Materials/MaterialFuctions/MF_6VertexPolygon}
  n5 = NamedRerouteDeclaration(None=OneMinus(None=SmoothStep(Value=@n4.SDF) {constMax=0.1})) {name=IdleState, nodeColor=<rgba 0 0 0 1>}
  n6 = LinearInterpolate(
   A=Constant3Vector() {constant=<rgba 1 1 1 0>}
   B=VectorParameter() {parameterName=Color1, defaultValue=<rgba 0.7991 0 0.5029 1>}.RGB
   Alpha=ScalarParameter() {parameterName=AnimateCooldown}
  )
  n7 = TextureCoordinate()
  n8 = AppendVector(
   A=ComponentMask(None=@n7) {r=true}
   B=Add(A=ComponentMask(None=@n7) {g=true}, B=ScalarParameter() {defaultValue=-0.2, parameterName=SecondArrowPosition}) {constB=-0.2}
  )
  n9 = Constant() {r=0.8771}
  n10 = MaterialFunctionCall(
   UVs=MaterialFunctionCall(
    UVs=MaterialFunctionCall(UVs=@n8, "Rotation Angle (0-1)"=@n9) {materialFunction=&CustomRotator}.Rotated Values
    Texture Scale=ScalarParameter() {defaultValue=0.45, parameterName=Second Arrow Scale}
   ) {materialFunction=&ScaleUVsByCenter}.UVs
   vertex_0=Constant2Vector() {g=-0.448, r=0.3131}
   vertex_1=Constant2Vector() {g=0.064, r=0.0183}
   vertex_2=Constant2Vector() {g=0.3032, r=-0.448}
   vertex_3=Constant2Vector() {g=0.9232, r=-0.9215}
   vertex_4=Constant2Vector() {g=0.3337, r=0.2926}
   vertex_5=Constant2Vector() {g=-0.8888, r=0.9097}
  ) {materialFunction=&MF_6VertexPolygon}
  n11 = NamedRerouteDeclaration(None=OneMinus(None=SmoothStep(Value=@n10.SDF) {constMax=0.1})) {name=SecondArrow, nodeColor=<rgba 0 0 0 1>}
  n12 = VectorParameter() {parameterName=Color, defaultValue=<rgba 0.7991 0 0.5029 1>}
  n13 = Panner(Coordinate=TextureCoordinate()) {speedY=-1.75}
  n14 = Power(Base=Sine(None=ComponentMask(None=@n13) {g=true})) {constExponent=20}
  Multiply(
   A=Add(A=Multiply(A=@n5, B=@n6), B=Multiply(A=@n11, B=@n12.RGB))
   B=LinearInterpolate(B=Add(A=@n14) {constB=0.6}, Alpha=ScalarParameter() {parameterName=AnimateLines}) {constA=1}
  )
  n15 = NamedRerouteUsage()
  n16 = ScalarParameter() {defaultValue=1, parameterName=OpacityIdle}
  n17 = Multiply(A=NamedRerouteUsage(), B=ScalarParameter() {parameterName=OpacityFirstArrow})
  n18 = Multiply(A=NamedRerouteUsage(), B=ScalarParameter() {parameterName=OpacitySecondArrow})
  Multiply(
   A=LinearInterpolate(A=Multiply(A=@n15, B=@n16), B=Add(A=@n17, B=@n18), Alpha=ScalarParameter() {parameterName=AnimateCooldown})
   B=ScalarParameter() {defaultValue=1, parameterName=FinalOpacity}
  )
 ]
}
