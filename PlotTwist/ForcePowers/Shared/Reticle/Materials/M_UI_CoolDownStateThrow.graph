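# UI material for the Force power throw reticle icon, blending between idle and animated cooldown looks through AnimateCooldown.
# Mirrored six-vertex polygon SDFs draw the activated and idle throw shapes procedurally, with a bottom-half mask for the cooldown part.
# The cooldown color blend uses a tintable Color and a radial gradient, while opacity blends from a dimmed state and scales by FinalOpacity.
# A Panner signal builds scrolling sine bands, rotated oppositely on each horizontal half.
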Material /PlotTwist/ForcePowers/Shared/Reticle/Materials/M_UI_CoolDownStateThrow
graph {
 properties {blendMode=BLEND_Translucent, materialDomain=MD_UI, pixelDepthOffsetMode=PDOM_Legacy, usedShadingModels=Default Lit}
 nodes [
  n1 = Multiply(A=NamedRerouteUsage(), B=VectorParameter() {parameterName=Color, defaultValue=<rgba 1 0.007 0.007 1>}.RGB)
  n2 = MaterialFunctionCall(UV=MaterialFunctionCall(Texture Scale=Constant() {r=2}) {materialFunction=&ScaleUVsByCenter}.UVs) {materialFunction=&MF_UI_RadialGradient}
  LinearInterpolate(
   A=NamedRerouteUsage()
   B=Multiply(A=Multiply(A=@n1, B=@n2.Smooth), B=NamedRerouteUsage())
   Alpha=ScalarParameter() {parameterName=AnimateCooldown}
  )
  n3 = NamedRerouteUsage()
  n4 = ScalarParameter() {defaultValue=0.6, parameterName=IdlePart Opacity}
  Multiply(
   A=LinearInterpolate(A=Multiply(A=@n3, B=@n4) {constB=0.8}, B=NamedRerouteUsage(), Alpha=ScalarParameter() {parameterName=AnimateCooldown})
   B=ScalarParameter() {defaultValue=1, parameterName=FinalOpacity}
  )
  n5 = TextureCoordinate()
  n6 = ComponentMask(None=@n5) {r=true}
  n7 = Multiply(A=ComponentMask(None=@n5) {g=true}) {constB=2}
  n8 = ScalarParameter() {defaultValue=0.5, parameterName=PositionX}
  n9 = ScalarParameter() {defaultValue=-1, parameterName=PositionY}
  n10 = MaterialFunctionCall(
   UVs=MaterialFunctionCall!(UV=AppendVector(A=@n6, B=@n7), Offset=AppendVector(A=@n8, B=@n9), Rotation, Scale=Constant() {r=0.3}, MirrorY) {materialFunction=/FortUILibrary/MaterialFunctions/Basics/MF_UI_Symmetry_2x1_DEPRECATED}.Result
   Texture Scale=Constant() {r=1}
  ) {materialFunction=&ScaleUVsByCenter}
  n11 = SmoothStep(Value=MaterialFunctionCall!(
    UVs=@n10.UVs
    vertex_0=Constant2Vector() {g=-0.7451, r=-0.9434}
    vertex_1=Constant2Vector() {g=-0.9326, r=0.3096}
    vertex_2=Constant2Vector() {g=-0.3566, r=0.8914}
    vertex_3=Constant2Vector() {g=0.9371, r=0.6994}
    vertex_4=Constant2Vector() {r=0.3291}
    vertex_5=Constant2Vector() {g=-0.2514, r=0.224}
    Glow Min
    Glow Max
    Outline Thickness
   ) {materialFunction=/PlotTwist/ForcePowers/Shared/Reticle/Materials/MaterialFuctions/MF_6VertexPolygon}.SDF) {constMax=0.1}
  n12 = SmoothStep(Value=MaterialFunctionCall(
    UVs=@n10.UVs
    vertex_0=Constant2Vector() {g=-0.2057, r=-0.6983}
    vertex_1=Constant2Vector() {g=-0.16, r=0.0947}
    vertex_2=Constant2Vector() {g=0.6217, r=0.1302}
    vertex_3=Constant2Vector() {g=0.0685, r=-0.1326}
    vertex_4=Constant2Vector() {g=-0.064, r=-0.4114}
    vertex_5=Constant2Vector() {g=-0.1829, r=-0.6537}
   ) {materialFunction=&MF_6VertexPolygon}.SDF) {constMax=0.1}
  NamedRerouteDeclaration(None=Add(A=OneMinus(None=@n11), B=OneMinus(None=@n12))) {name=ThrowShape_Activated, nodeColor=<rgba 0 0 0 1>}
  n13 = TextureCoordinate()
  n14 = ComponentMask(None=@n13) {g=true}
  n15 = MaterialFunctionCall(
   UV=AppendVector(A=ComponentMask(None=@n13) {r=true}, B=Multiply(A=@n14) {constB=2})
   Offset=Constant2Vector() {r=0.15}
   Scale=Constant() {r=0.25}
  ) {materialFunction=&MF_UI_Symmetry_2x1_DEPRECATED}
  n16 = Constant() {r=1}
  n17 = SmoothStep(Value=MaterialFunctionCall(
    UVs=MaterialFunctionCall(UVs=@n15.Result, Texture Scale=@n16) {materialFunction=&ScaleUVsByCenter}.UVs
    vertex_0=Constant2Vector() {g=-0.6171, r=-0.9115}
    vertex_1=Constant2Vector() {g=-0.8686, r=0.5381}
    vertex_2=Constant2Vector() {g=-0.5486, r=0.864}
    vertex_3=Constant2Vector() {g=0.9371, r=0.5806}
    vertex_4=Constant2Vector() {r=0.2834}
    vertex_5=Constant2Vector() {g=-0.2331, r=0.2011}
   ) {materialFunction=&MF_6VertexPolygon}.SDF) {constMax=0.1}
  NamedRerouteDeclaration(None=Multiply(A=OneMinus(None=@n17), B=NamedRerouteUsage())) {name=IdlePart, nodeColor=<rgba 0 0 0 1>}
  n18 = Multiply(A=TextureCoordinate(), B=Constant2Vector() {g=2, r=2})
  n19 = Saturate(None=Floor(None=ComponentMask(None=@n18) {g=true}))
  NamedRerouteDeclaration(None=Multiply(A=NamedRerouteUsage(), B=NamedRerouteDeclaration(None=@n19) {name=BottomMask, nodeColor=<rgba 0 0 0 1>})) {name=CooldownPart, nodeColor=<rgba 0 0 0 1>}
  n20 = TextureCoordinate()
  n21 = Constant() {r=-0.125}
  n22 = ComponentMask(None=Panner(Coordinate=MaterialFunctionCall(UVs=@n20, "Rotation Angle (0-1)"=@n21) {materialFunction=&CustomRotator}.Rotated Values) {speedY=1}) {g=true}
  n23 = Add(A=Power(Base=Sine(None=@n22)) {constExponent=20}) {constB=0.6}
  n24 = ComponentMask(None=Multiply(A=TextureCoordinate(), B=Constant2Vector() {g=2, r=2})) {r=true}
  n25 = OneMinus(None=Saturate(None=Floor(None=@n24)))
  n26 = TextureCoordinate()
  n27 = Constant() {r=0.125}
  n28 = ComponentMask(None=Panner(Coordinate=MaterialFunctionCall(UVs=@n26, "Rotation Angle (0-1)"=@n27) {materialFunction=&CustomRotator}.Rotated Values) {speedY=1}) {g=true}
  n29 = Add(A=Power(Base=Sine(None=@n28)) {constExponent=20}) {constB=0.6}
  n30 = Multiply(A=TextureCoordinate(), B=Constant2Vector() {g=2, r=2})
  n31 = Saturate(None=Floor(None=ComponentMask(None=@n30) {r=true}))
  NamedRerouteDeclaration(None=Add(A=Multiply(A=@n23, B=@n25), B=Multiply(A=@n29, B=@n31))) {name=Panner, nodeColor=<rgba 0 0 0 1>}
 ]
}
