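# Reticle extension widget for PlotTwist Force powers (Push, Rip, Projectile, AmberVine, HeavyLamp), showing targeting, charges and cooldown.
# Construct waits for passive tags and binds each ability's reticle update, which fades lock-on by target range.
# Tick polls pawn tags to select power images, update charge pips and drive cooldown animations and material parameters, including AmberVine channel time.
# Counters make state-change animations fire once rather than every update.
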WidgetBlueprint /PlotTwist/ForcePowers/Shared/Reticle/RE_PlotTwist_ForcePowers_ReticleExtension
animations [
 Homing {
  binding_count=1
  length_seconds=0.1499
  object=&self:Homing
  sequence {
   bindings [Reticle {
     id=#1
     possessed_class=Image
     tracks [{
       type=MovieSceneWidgetMaterialTrack
       sections [{
         type=MovieSceneParameterSection
         end_frame=4
         start_frame=0
         channels [
          LockedOn {type=MovieSceneScriptingFloatChannel, keys [[0,0],[4.3755,1]]}
          AnimateScale {type=MovieSceneScriptingFloatChannel, keys [[0,0],[4.3755,1]]}
         ]
        }]
      }]
    }]
   display_rate {numerator=30}
  }
 }
 Anim_Push {
  binding_count=2
  length_seconds=2.25
  object=&self:Anim_Push
  sequence {bindings [
    Reticle_Push {
     id=#2
     possessed_class=Image
     tracks [
      {
       type=MovieSceneWidgetMaterialTrack
       sections [{
         type=MovieSceneParameterSection
         end_frame=45
         start_frame=0
         channels [
          IdlePartOpacity {type=MovieSceneScriptingFloatChannel, keys [[5,1],[10,0]]}
          AnimateCooldown {type=MovieSceneScriptingFloatChannel, keys [[10,0],[15,1]]}
          GradientStartMiddle {type=MovieSceneScriptingFloatChannel, keys [[15,0.001],[30,2]]}
          GradientStartSides {type=MovieSceneScriptingFloatChannel, keys [[25,0.001],[40,2]]}
         ]
        }]
      }
      {
       type=MovieScene2DTransformTrack
       sections [{
         type=MovieScene2DTransformSection
         end_frame=45
         start_frame=15
         channels [
          Translation.X {type=MovieSceneScriptingFloatChannel}
          Translation.Y {type=MovieSceneScriptingFloatChannel}
          Angle {type=MovieSceneScriptingFloatChannel}
          Scale.X {type=MovieSceneScriptingFloatChannel}
          Scale.Y {type=MovieSceneScriptingFloatChannel}
          Shear.X {type=MovieSceneScriptingFloatChannel}
          Shear.Y {type=MovieSceneScriptingFloatChannel}
         ]
        }]
      }
     ]
    }
    Reticle {
     id=#3
     possessed_class=Image
     tracks [{
       type=MovieSceneWidgetMaterialTrack
       sections [{
         type=MovieSceneParameterSection
         end_frame=45
         start_frame=0
         channels [AnimateTopPart {type=MovieSceneScriptingFloatChannel, keys [[0,0],[5,1]]}]
        }]
      }]
    }
   ]}
 }
 AnimThrow {
  binding_count=2
  length_seconds=1.4
  object=&self:AnimThrow
  sequence {bindings [
    Reticle_Throw {
     id=#4
     possessed_class=Image
     tracks [
      {
       type=MovieSceneWidgetMaterialTrack
       sections [{
         type=MovieSceneParameterSection
         end_frame=28
         start_frame=0
         channels [
          AnimateCooldown {type=MovieSceneScriptingFloatChannel, keys [[0,0],[10,1]]}
          PositionY {type=MovieSceneScriptingFloatChannel, keys [[12,-0.5646],[17,-0.02]]}
          PositionX {type=MovieSceneScriptingFloatChannel, keys [[12,0.65],[17,0.16]]}
          FinalOpacity {type=MovieSceneScriptingFloatChannel, const [1,1]}
         ]
        }]
      }
      {
       type=MovieScene2DTransformTrack
       sections [{
         type=MovieScene2DTransformSection
         end_frame=28
         start_frame=0
         channels [
          Translation.X {type=MovieSceneScriptingFloatChannel}
          Translation.Y {type=MovieSceneScriptingFloatChannel}
          Angle {type=MovieSceneScriptingFloatChannel}
          Scale.X {type=MovieSceneScriptingFloatChannel, keys [[0,1],[11,1],[12,1.2],[17,1]]}
          Scale.Y {type=MovieSceneScriptingFloatChannel, keys [[0,1],[11,1],[12,1.2],[17,1]]}
          Shear.X {type=MovieSceneScriptingFloatChannel}
          Shear.Y {type=MovieSceneScriptingFloatChannel}
         ]
        }]
      }
     ]
    }
    Reticle {
     id=#5
     possessed_class=Image
     tracks [{
       type=MovieSceneWidgetMaterialTrack
       sections [{
         type=MovieSceneParameterSection
         end_frame=25
         start_frame=0
         channels [AnimateTopPart {type=MovieSceneScriptingFloatChannel, keys [[0,0],[5,1]]}]
        }]
      }]
    }
   ]}
 }
 Anim_Pull {
  binding_count=2
  length_seconds=1.3
  object=&self:Anim_Pull
  sequence {bindings [
    Reticle_Pull {
     id=#6
     possessed_class=Image
     tracks [
      {
       type=MovieSceneWidgetMaterialTrack
       sections [{
         type=MovieSceneParameterSection
         end_frame=26
         start_frame=0
         channels [
          OpacityIdle {type=MovieSceneScriptingFloatChannel, keys [[0,1],[10,0]]}
          AnimateCooldown {type=MovieSceneScriptingFloatChannel, keys [[10,0],[14,1]]}
          OpacityFirstArrow {type=MovieSceneScriptingFloatChannel, keys [[14,0],[20,1]]}
          FirstArrowPosition {type=MovieSceneScriptingFloatChannel, keys [[0,-0.1],[10,-0.1],[14,0],[20,-0.1]]}
          OpacitySecondArrow {type=MovieSceneScriptingFloatChannel, keys [[20,0],[25,1]]}
          SecondArrowPosition {type=MovieSceneScriptingFloatChannel, keys [[20,0],[25,-0.18]]}
          AnimateLines {type=MovieSceneScriptingFloatChannel, keys [[25,0],[26,1]]}
          Second Arrow Scale {type=MovieSceneScriptingFloatChannel, keys [[20,0.25],[25,0.45]]}
         ]
        }]
      }
      {
       type=MovieSceneColorTrack
       display_name=Color and Opacity
       sections [{
         type=MovieSceneColorSection
         end_frame=5
         start_frame=0
         channels [
          Color.R {type=MovieSceneScriptingFloatChannel}
          Color.G {type=MovieSceneScriptingFloatChannel}
          Color.B {type=MovieSceneScriptingFloatChannel}
          Color.A {type=MovieSceneScriptingFloatChannel, keys [[0,0.6],[5,1]]}
         ]
        }]
      }
     ]
    }
    Reticle {
     id=#7
     possessed_class=Image
     tracks [{
       type=MovieSceneWidgetMaterialTrack
       sections [{
         type=MovieSceneParameterSection
         end_frame=26
         start_frame=0
         channels [AnimateTopPart {type=MovieSceneScriptingFloatChannel, keys [[0,0],[5,1]]}]
        }]
      }]
    }
   ]}
 }
 Anim_CoolDownState {
  binding_count=4
  length_seconds=0.25
  object=&self:Anim_CoolDownState
  sequence {bindings [
    Reticle {
     id=#8
     possessed_class=Image
     tracks [{
       type=MovieSceneWidgetMaterialTrack
       sections [{
         type=MovieSceneParameterSection
         end_frame=5
         start_frame=0
         channels [
          FinalOpacity {type=MovieSceneScriptingFloatChannel, keys [[0,1],[5,0.4]]}
          AnimCooldown {type=MovieSceneScriptingFloatChannel, keys [[0,1],[5,0]]}
         ]
        }]
      }]
    }
    Reticle_Pull {
     id=#9
     possessed_class=Image
     tracks [
      {
       type=MovieSceneWidgetMaterialTrack
       sections [{
         type=MovieSceneParameterSection
         end_frame=5
         start_frame=0
         channels [FinalOpacity {type=MovieSceneScriptingFloatChannel, keys [[0,1],[4.9997,6],[5,0.4]]}]
        }]
      }
      {
       type=MovieScene2DTransformTrack
       sections [{
         type=MovieScene2DTransformSection
         end_frame=5
         start_frame=0
         channels [
          Translation.X {type=MovieSceneScriptingFloatChannel, const [0,1]}
          Translation.Y {type=MovieSceneScriptingFloatChannel, keys [[0,4],[5,8]]}
          Angle {type=MovieSceneScriptingFloatChannel, const [0,1]}
          Scale.X {type=MovieSceneScriptingFloatChannel, const [1,1]}
          Scale.Y {type=MovieSceneScriptingFloatChannel, const [1,1]}
          Shear.X {type=MovieSceneScriptingFloatChannel, const [0,1]}
          Shear.Y {type=MovieSceneScriptingFloatChannel, const [0,1]}
         ]
        }]
      }
     ]
    }
    Reticle_Throw {
     id=#10
     possessed_class=Image
     tracks [
      {
       type=MovieSceneWidgetMaterialTrack
       sections [{
         type=MovieSceneParameterSection
         end_frame=5
         start_frame=0
         channels [FinalOpacity {type=MovieSceneScriptingFloatChannel, keys [[0,1],[4.9997,0.6],[5,0.4]]}]
        }]
      }
      {
       type=MovieScene2DTransformTrack
       sections [{
         type=MovieScene2DTransformSection
         end_frame=5
         start_frame=0
         channels [
          Translation.X {type=MovieSceneScriptingFloatChannel, const [0,1]}
          Translation.Y {type=MovieSceneScriptingFloatChannel, keys [[0,0],[5,4]]}
          Angle {type=MovieSceneScriptingFloatChannel, const [0,1]}
          Scale.X {type=MovieSceneScriptingFloatChannel, const [1,1]}
          Scale.Y {type=MovieSceneScriptingFloatChannel, const [1,1]}
          Shear.X {type=MovieSceneScriptingFloatChannel, const [0,1]}
          Shear.Y {type=MovieSceneScriptingFloatChannel, const [0,1]}
         ]
        }]
      }
     ]
    }
    Reticle_Push {
     id=#11
     possessed_class=Image
     tracks [
      {
       type=MovieSceneWidgetMaterialTrack
       sections [{
         type=MovieSceneParameterSection
         end_frame=5
         start_frame=0
         channels [FinalOpacity {type=MovieSceneScriptingFloatChannel, keys [[0,1],[4.9997,0.6],[5,0.4]]}]
        }]
      }
      {
       type=MovieScene2DTransformTrack
       sections [{
         type=MovieScene2DTransformSection
         end_frame=5
         start_frame=0
         channels [
          Translation.X {type=MovieSceneScriptingFloatChannel, const [0,1]}
          Translation.Y {type=MovieSceneScriptingFloatChannel, keys [[0,0],[5,4]]}
          Angle {type=MovieSceneScriptingFloatChannel, const [0,1]}
          Scale.X {type=MovieSceneScriptingFloatChannel, const [1,1]}
          Scale.Y {type=MovieSceneScriptingFloatChannel, const [1,1]}
          Shear.X {type=MovieSceneScriptingFloatChannel, const [0,1]}
          Shear.Y {type=MovieSceneScriptingFloatChannel, const [0,1]}
         ]
        }]
      }
     ]
    }
   ]}
 }
 Anim_LockOn {
  binding_count=1
  length_seconds=0.25
  object=&self:Anim_LockOn
  sequence {bindings [ReticleOverlay {
     id=#12
     possessed_class=Overlay
     tracks [{
       type=MovieSceneFloatTrack
       display_name=Render Opacity
       sections [{type=MovieSceneFloatSection, end_frame=5, start_frame=0, channels [None {type=MovieSceneScriptingFloatChannel, keys [[0,0.4],[5,1]]}]}]
      }]
    }]}
 }
 Anim_Lightning {
  binding_count=2
  length_seconds=2.25
  object=&self:Anim_Lightning
  sequence {bindings [
    Reticle_Lightning {
     id=#13
     possessed_class=Image
     tracks [
      {
       type=MovieSceneWidgetMaterialTrack
       sections [{
         type=MovieSceneParameterSection
         end_frame=45
         start_frame=0
         channels [
          IdlePartOpacity {type=MovieSceneScriptingFloatChannel, keys [[5,1],[10,0]]}
          AnimateCooldown {type=MovieSceneScriptingFloatChannel, keys [[10,0],[15,1]]}
          GradientStartMiddle {type=MovieSceneScriptingFloatChannel, keys [[15,0.001],[30,2]]}
          GradientStartSides {type=MovieSceneScriptingFloatChannel, keys [[25,0.001],[40,2]]}
         ]
        }]
      }
      {
       type=MovieScene2DTransformTrack
       sections [{
         type=MovieScene2DTransformSection
         end_frame=45
         start_frame=15
         channels [
          Translation.X {type=MovieSceneScriptingFloatChannel}
          Translation.Y {type=MovieSceneScriptingFloatChannel}
          Angle {type=MovieSceneScriptingFloatChannel}
          Scale.X {type=MovieSceneScriptingFloatChannel}
          Scale.Y {type=MovieSceneScriptingFloatChannel}
          Shear.X {type=MovieSceneScriptingFloatChannel}
          Shear.Y {type=MovieSceneScriptingFloatChannel}
         ]
        }]
      }
     ]
    }
    Reticle {
     id=#3
     possessed_class=Image
     tracks [{
       type=MovieSceneWidgetMaterialTrack
       sections [{
         type=MovieSceneParameterSection
         end_frame=45
         start_frame=0
         channels [AnimateTopPart {type=MovieSceneScriptingFloatChannel, keys [[0,0],[5,1]]}]
        }]
      }]
    }
   ]}
 }
 Anim_Use {
  binding_count=1
  length_seconds=0.5
  object=&self:Anim_Use
  sequence {bindings [Reticle_Lightning {
     id=#14
     possessed_class=Image
     tracks [{
       type=MovieSceneWidgetMaterialTrack
       sections [{
         type=MovieSceneParameterSection
         end_frame=10
         start_frame=0
         channels [AnimatePipProgress {type=MovieSceneScriptingFloatChannel, keys [[0,1],[10,0]]}]
        }]
      }]
    }]}
 }
 Anim_ShowVFX {
  binding_count=1
  length_seconds=2
  object=&self:Anim_ShowVFX
  sequence {bindings [Reticle_Lightning_VFX {
     id=#15
     possessed_class=Image
     tracks [{
       type=MovieSceneFloatTrack
       display_name=Render Opacity
       sections [{
         type=MovieSceneFloatSection
         end_frame=40
         start_frame=0
         channels [None {type=MovieSceneScriptingFloatChannel, keys [[0,1],[34,1],[39.9993,0]]}]
        }]
      }]
    }]}
 }
]
graphs [
 EventGraph {nodes [
   BindReticle_HeavyLamp() {is=title K2Node_CustomEvent}
   > n1 = GetPrimaryAbilityInstanceOfAbilitySystemComponent!(
    execute
    AbilitySystemComponent:Fort Ability System Component Object Reference=GetFortAbilitySystemComponent!(Actor:Actor Object Reference=Get PlayerPawn() -> (PlayerPawn:Fort Player Pawn Athena Object Reference)) -> (ReturnValue:Fort Ability System Component Object Reference)
    AbilityClass:Fort Gameplay Ability Class Reference=/PlotTwist/Items/HeavyLamp/Abilities/GA_Athena_PlotTwist_HeavyLamp_Passive.GA_Athena_PlotTwist_HeavyLamp_Passive_C
   ) -> (then, ReturnValue:Gameplay Ability Object Reference)
   > Is Valid(InputObject=@n1)
   > n2 = Cast To GA_Athena_PlotTwist_HeavyLamp_Passive!(execute, Object=@n1) -> (then, CastFailed, AsGA Athena Plot Twist Heavy Lamp Passive:GA Athena Plot Twist Heavy Lamp Passive Object Reference) {is=title K2Node_DynamicCast}
   > Bind Event to Update Reticle(self:GA Athena Plot Twist Force Passive Parent Object Reference=@n2, Delegate=Create Event()) {is=title K2Node_AddDelegate}
   TEMP_SetupReticleCharges() {is=title K2Node_CustomEvent}
   > Branch(Condition=Get Temp_ShowChargesEnabled() -> (Temp_ShowChargesEnabled:bool))
   > Set AmmoQuantity(AmmoQuantity:int=-1) -> (Output_Get:int=0)
   SetAmmoItemDefinition() {is=title K2Node_CustomEvent}
   > n3 = Branch(Condition=HasMatchingGameplayTag!(
     self:Gameplay Tag Asset Interface Interface=Get PlayerPawn()
     TagToCheck:Gameplay Tag Structure=(TagName="Granted.Athena.PlotTwist.Force.Power.Projectile")
    ) -> (ReturnValue:bool))
   > Set AmmoItemDefinition(AmmoItemDefinition:Fort Item Definition Object Reference=Get Ammo_Projectile() -> (Ammo_Projectile:Fort Item Definition Object Reference)) -> (Output_Get:Fort Item Definition Object Reference)
   n4 = Branch(execute=@n3.else, Condition=HasMatchingGameplayTag(self=Get PlayerPawn(), TagToCheck=(TagName="Granted.Athena.PlotTwist.Force.Power.Push")))
   > Set AmmoItemDefinition(AmmoItemDefinition=Get Ammo_Push() -> (Ammo_Push:Fort Item Definition Object Reference))
   n5 = Branch(execute=@n4.else, Condition=HasMatchingGameplayTag(self=Get PlayerPawn(), TagToCheck=(TagName="Granted.Athena.PlotTwist.Force.Power.Rip")))
   > Set AmmoItemDefinition(AmmoItemDefinition=Get Ammo_Rip() -> (Ammo_Rip:Fort Item Definition Object Reference))
   n6 = Branch(
    execute=@n5.else
    Condition=HasMatchingGameplayTag(self=Get PlayerPawn(), TagToCheck=Get Tag_PassiveGranted_AmberVine() -> (Tag_PassiveGranted_AmberVine:Gameplay Tag Structure))
   )
   > Set AmmoItemDefinition(AmmoItemDefinition=Get Ammo_AmberVine() -> (Ammo_AmberVine:Fort Item Definition Object Reference))
   n7 = Branch(
    execute=@n6.else
    Condition=HasMatchingGameplayTag(self=Get PlayerPawn(), TagToCheck=Get Tag_Ability_HeavyLamp() -> (Tag_Ability_HeavyLamp:Gameplay Tag Structure))
   )
   > Set AmmoItemDefinition(AmmoItemDefinition=Get AmmoHeavyLamp() -> (AmmoHeavyLamp:Fort Item Definition Object Reference))
   Set AmmoItemDefinition(execute=@n7.else)
   UpdateAmberVineComponent() {is=title K2Node_CustomEvent}
   > n8 = Get!(execute:Exec) -> (then:Exec, else:Exec, AmberVine Component:BP Ambervine Pawn Component Object Reference)
   n9 = GetChannelTimeRemaining!(
    execute=[@n12.Is Valid, @n8.then]
    self:BP Ambervine Pawn Component Object Reference=Get AmberVine Component() -> (AmberVine Component:BP Ambervine Pawn Component Object Reference)
   ) -> (then, IsChanneling:bool, PercentageTimeRemaining:double)
   > Branch(Condition=@n9)
   > SetScalarParameterValue(
    self=Get AmberVineCooldownMaterialRef() -> (AmberVineCooldownMaterialRef:Material Instance Dynamic Object Reference)
    ParameterName=AnimatePipProgress
    Value=@n9.PercentageTimeRemaining
   )
   n10 = Get!(execute=@n8.else) -> (then, else, PlayerPawn:Fort Player Pawn Athena Object Reference)
   > n11 = Set AmberVine Component(AmberVine Component:BP Ambervine Pawn Component Object Reference=GetComponentByClass!(
     self:Actor Object Reference=@n10
     ComponentClass:Actor Component Class Reference=/PlotTwist/ForcePowers/AmberVine/BP_AmbervinePawnComponent.BP_AmbervinePawnComponent_C
    ) -> (ReturnValue:BP Ambervine Pawn Component Object Reference)) -> (Output_Get:BP Ambervine Pawn Component Object Reference)
   > n12 = Is Valid(InputObject=@n11)
   n13 = Event_UpdateReticle!() -> (OutputDelegate, then, TargetLocationIsInRange:bool=~) {is=title K2Node_CustomEvent}
   > n14 = Set Target Location Is in Range(Target Location Is in Range:bool=@n13.TargetLocationIsInRange) -> (Output_Get:bool)
   > n15 = Branch(Condition=@n14)
   > Set CounterPlayLockOnBackward(CounterPlayLockOnBackward:double=0) -> (Output_Get:double)
   > n16 = Increment Float!(" ":Exec, Value:"Float (double-precision) (by ref)"=Get CounterPlayLockOnForward() -> (CounterPlayLockOnForward:double)) -> ("  ":Exec, Result:double=~) {is=title K2Node_MacroInstance}
   > n17 = Set CounterPlayLockOnForward(CounterPlayLockOnForward:double=@n16) -> (Output_Get:double)
   > n18 = Branch(Condition="float > float"(A=@n17, B=1))
   n19 = Get Anim_LockOn() -> (Anim_LockOn:Widget Animation Object Reference)
   PlayAnimationForward(execute=@n18.else, InAnimation=@n19)
   Set CounterPlayLockOnForward(execute=@n15.else, CounterPlayLockOnForward=0)
   > n20 = Increment Float(Value=Get CounterPlayLockOnBackward() -> (CounterPlayLockOnBackward:double)) -> (Result=~)
   > n21 = Set CounterPlayLockOnBackward(CounterPlayLockOnBackward=@n20)
   > n22 = Branch(Condition="float > float"(A=@n21, B=1))
   PlayAnimationReverse(execute=@n22.else, InAnimation=@n19)
   BindReticle_Push() {is=title K2Node_CustomEvent}
   > n23 = GetPrimaryAbilityInstanceOfAbilitySystemComponent(
    AbilitySystemComponent=GetFortAbilitySystemComponent(Actor=Get PlayerPawn())
    AbilityClass=/PlotTwist/ForcePowers/ForcePush/GA_Athena_PlotTwist_Force_Push_Passive.GA_Athena_PlotTwist_Force_Push_Passive_C
   )
   > Is Valid(InputObject=@n23)
   > n24 = Cast To GA_Athena_PlotTwist_Force_Push_Passive!(execute, Object=@n23) -> (then, CastFailed, AsGA Athena Plot Twist Force Push Passive:GA Athena Plot Twist Force Push Passive Object Reference) {is=title K2Node_DynamicCast}
   > Bind Event to Update Reticle(self=@n24, Delegate=Create Event())
   BindReticle_Rip() {is=title K2Node_CustomEvent}
   > n25 = GetPrimaryAbilityInstanceOfAbilitySystemComponent(
    AbilitySystemComponent=GetFortAbilitySystemComponent(Actor=Get PlayerPawn())
    AbilityClass=/PlotTwist/ForcePowers/ForceRip/GA_Athena_PlotTwist_Force_Rip_Passive.GA_Athena_PlotTwist_Force_Rip_Passive_C
   )
   > Is Valid(InputObject=@n25)
   > n26 = Cast To GA_Athena_PlotTwist_Force_Rip_Passive!(execute, Object=@n25) -> (then, CastFailed, AsGA Athena Plot Twist Force Rip Passive:GA Athena Plot Twist Force Rip Passive Object Reference) {is=title K2Node_DynamicCast}
   > Bind Event to Update Reticle(self=@n26, Delegate=Create Event())
   BindReticle_AmberVine() {is=title K2Node_CustomEvent}
   > n27 = GetPrimaryAbilityInstanceOfAbilitySystemComponent(
    AbilitySystemComponent=GetFortAbilitySystemComponent(Actor=Get PlayerPawn())
    AbilityClass=/PlotTwist/ForcePowers/AmberVine/GA_Athena_AmberVine_Passive.GA_Athena_AmberVine_Passive_C
   )
   > Is Valid(InputObject=@n27)
   > n28 = Cast To GA_Athena_AmberVine_Passive!(execute, Object=@n27) -> (then, CastFailed, AsGA Athena Amber Vine Passive:GA Athena Amber Vine Passive Object Reference) {is=title K2Node_DynamicCast}
   > Bind Event to Update Reticle(self=@n28, Delegate=Create Event())
   BindReticle_Projectile() {is=title K2Node_CustomEvent}
   > n29 = GetPrimaryAbilityInstanceOfAbilitySystemComponent(
    AbilitySystemComponent=GetFortAbilitySystemComponent(Actor=Get PlayerPawn())
    AbilityClass=/PlotTwist/ForcePowers/ForceProjectile/GA_Athena_PlotTwist_Force_Projectile_Passive.GA_Athena_PlotTwist_Force_Projectile_Passive_C
   )
   > Is Valid(InputObject=@n29)
   > n30 = Cast To GA_Athena_PlotTwist_Force_Projectile_Passive!(execute, Object=@n29) -> (then, CastFailed, AsGA Athena Plot Twist Force Projectile Passive:GA Athena Plot Twist Force Projectile Passive Object Reference) {is=title K2Node_DynamicCast}
   > Bind Event to Update Reticle(self=@n30, Delegate=Create Event())
   Event Destruct()
   > EndAction(self:Ability Async Object Reference=[
     Get Task_WaitPassiveGranted_Push() -> (Task_WaitPassiveGranted_Push:Ability Async Wait Gameplay Tag Added Object Reference)
     Get Task_WaitPassiveGranted_Rip() -> (Task_WaitPassiveGranted_Rip:Ability Async Wait Gameplay Tag Added Object Reference)
     Get Task_WaitPassiveGranted_Projectile() -> (Task_WaitPassiveGranted_Projectile:Ability Async Wait Gameplay Tag Added Object Reference)
     Get Task_WaitPassiveGranted_AmberVine() -> (Task_WaitPassiveGranted_AmberVine:Ability Async Wait Gameplay Tag Added Object Reference)
     Get Task_WaitPassiveGranted_HeavyLamp() -> (Task_WaitPassiveGranted_HeavyLamp:Ability Async Wait Gameplay Tag Added Object Reference)
    ])
   Event Construct()
   > n31 = GetDynamicMaterial(self=Get Reticle_Lightning() -> (Reticle_Lightning:Image Object Reference))
   > Set AmberVineCooldownMaterialRef(AmberVineCooldownMaterialRef:Material Instance Dynamic Object Reference=@n31) -> (Output_Get:Material Instance Dynamic Object Reference)
   > n32 = Cast To FortPlayerPawnAthena(Object=GetOwningPlayerPawn!(self:User Widget Object Reference) -> (ReturnValue:Pawn Object Reference))
   > n33 = Set PlayerPawn(PlayerPawn:Fort Player Pawn Athena Object Reference=@n32) -> (Output_Get:Fort Player Pawn Athena Object Reference)
   > Set AmberVine Component(AmberVine Component=GetComponentByClass(self=@n33, ComponentClass=&BP_AmbervinePawnComponent_C))
   > n34 = Set PlayerController(PlayerController:Fort Player Controller Object Reference=GetFortPlayerControllerFromActor!(Actor:Actor Object Reference=@n32) -> (ReturnValue:Fort Player Controller Object Reference)) -> (Output_Get:Fort Player Controller Object Reference)
   > Is Valid(InputObject=@n34)
   > Set Target Location Is in Range(Target Location Is in Range=true)
   > n35 = Sequence!(execute) -> (then_0, then_1, then_2:Exec, then_3:Exec, then_4:Exec)
   > n36 = WaitGameplayTagAddToActor!(
    execute:Exec
    TargetActor:Actor Object Reference=Get PlayerPawn()
    Tag:Gameplay Tag Structure=Get Tag_PassiveGranted_Push() -> (Tag_PassiveGranted_Push:Gameplay Tag Structure)
    OnlyTriggerOnce:bool=false
   ) -> (then:Exec, AsyncTaskProxy:Ability Async Wait Gameplay Tag Added Object Reference, Added:Exec) {is=title K2Node_AsyncAction}
   > Set Task_WaitPassiveGranted_Push(Task_WaitPassiveGranted_Push:Ability Async Wait Gameplay Tag Added Object Reference=@n36) -> (Output_Get:Ability Async Wait Gameplay Tag Added Object Reference)
   BindReticle_Push(execute=@n36.Added) {is=title K2Node_CallFunction}
   n37 = WaitGameplayTagAddToActor(
    execute=@n35.then_2
    TargetActor=Get PlayerPawn()
    Tag=Get Tag_PassiveGranted_Projectile() -> (Tag_PassiveGranted_Projectile:Gameplay Tag Structure)
    OnlyTriggerOnce=false
   )
   > Set Task_WaitPassiveGranted_Projectile(Task_WaitPassiveGranted_Projectile:Ability Async Wait Gameplay Tag Added Object Reference=@n37) -> (Output_Get:Ability Async Wait Gameplay Tag Added Object Reference)
   BindReticle_Projectile(execute=@n37.Added) {is=title K2Node_CallFunction}
   n38 = WaitGameplayTagAddToActor(execute=@n35.then_3, TargetActor=Get PlayerPawn(), Tag=Get Tag_PassiveGranted_AmberVine(), OnlyTriggerOnce=false)
   > Set Task_WaitPassiveGranted_AmberVine(Task_WaitPassiveGranted_AmberVine:Ability Async Wait Gameplay Tag Added Object Reference=@n38) -> (Output_Get:Ability Async Wait Gameplay Tag Added Object Reference)
   BindReticle_AmberVine(execute=@n38.Added) {is=title K2Node_CallFunction}
   n39 = WaitGameplayTagAddToActor(
    execute=@n35.then_1
    TargetActor=Get PlayerPawn()
    Tag=Get Tag_PassiveGranted_Rip() -> (Tag_PassiveGranted_Rip:Gameplay Tag Structure)
    OnlyTriggerOnce=false
   )
   > Set Task_WaitPassiveGranted_Rip(Task_WaitPassiveGranted_Rip:Ability Async Wait Gameplay Tag Added Object Reference=@n39) -> (Output_Get:Ability Async Wait Gameplay Tag Added Object Reference)
   BindReticle_Rip(execute=@n39.Added) {is=title K2Node_CallFunction}
   n40 = WaitGameplayTagAddToActor(
    execute=@n35.then_4
    TargetActor=Get PlayerPawn()
    Tag=Get Tag_PassiveGranted_HeavyLamp() -> (Tag_PassiveGranted_HeavyLamp:Gameplay Tag Structure)
    OnlyTriggerOnce=false
   )
   > Set Task_WaitPassiveGranted_HeavyLamp(Task_WaitPassiveGranted_HeavyLamp:Ability Async Wait Gameplay Tag Added Object Reference=@n40) -> (Output_Get:Ability Async Wait Gameplay Tag Added Object Reference)
   BindReticle_HeavyLamp(execute=@n40.Added) {is=title K2Node_CallFunction}
   Event Tick()
   > UpdateAmberVineComponent() {is=title K2Node_CallFunction}
   > n41 = Set ShowReticle(ShowReticle:bool=HasAnyMatchingGameplayTags!(
     self:Gameplay Tag Asset Interface Interface=Get PlayerPawn()
     TagContainer:"Gameplay Tag Container Structure (by ref)"=Get Tag_ShowReticle() -> (Tag_ShowReticle:Gameplay Tag Container Structure)
    ) -> (ReturnValue:bool)) -> (Output_Get:bool)
   > n42 = Branch(Condition=@n41)
   > SetVisibility(self=Get Reticle() -> (Reticle:Image Object Reference), InVisibility=Visible)
   > n43 = SetReticleCooldown()
   n44 = Get!(execute=[@n43.then, @n73.then]) -> (then, else, PlayerController:Fort Player Controller Object Reference)
   > SetAmmoItemDefinition() {is=title K2Node_CallFunction}
   > n45 = Get!(execute) -> (then, else, AmmoItemDefinition:Fort Item Definition Object Reference)
   n46 = GetFortAbilitySystemComponent(Actor=Get PlayerPawn())
   > n47 = Branch(Condition="integer > integer"(
     A=GetMaxStackSize!(self:Fort Item Definition Object Reference=@n45, AbilityComp:Ability System Component Object Reference=@n46) -> (ReturnValue:int=0)
     B=1
    ))
   > n48 = Set AmmoQuantity(AmmoQuantity=Get Item Quantity On Player!(
     PlayerController:Controller Object Reference=Get PlayerController() -> (PlayerController:Fort Player Controller Object Reference)
     ItemDefinition:Item Definition Base Object Reference=@n45
     ItemVariantGuid:Guid Structure
    ) -> (ReturnValue:int=0) {is=title K2Node_CallFunction}) -> (Output_Get=0)
   n49 = For Each Loop(Exec=[@n48.then, @n72.then], Array:Array of Image Object References=Get Charges() -> (Charges:Array of Image Object References)) -> (Array Element:Image Object Reference)
   n50 = int + int(A=@n49.Array Index)
   n51 = Get AmmoQuantity() -> (AmmoQuantity:int=0)
   > SetOpacity!(execute, self:Image Object Reference=@n49, InOpacity:float=SelectFloat(A=1, B=0, bPickA="integer <= integer"(A=@n50, B=@n51))) -> (then)
   n52 = GetDynamicMaterial(execute=@n49.Completed, self=Get Reticle())
   n53 = HasMatchingGameplayTag(self=Get PlayerPawn(), TagToCheck=(TagName="Granted.Athena.PlotTwist.Force.Power.Push"))
   n54 = Get PlayerPawn()
   > SetScalarParameterValue(
    self=@n52
    ParameterName=ReticleBottomPart
    Value=SelectFloat(
     A=1
     B=SelectFloat(A=0.5, B=0, bPickA=@n53)
     bPickA=HasMatchingGameplayTag(self=@n54, TagToCheck=(TagName="Granted.Athena.PlotTwist.Force.Power.Rip"))
    )
   )
   > n55 = Branch(Condition=HasAnyMatchingGameplayTags(self=Get PlayerPawn(), TagContainer=Get Tag_CooldownActive() -> (Tag_CooldownActive:Gameplay Tag Container Structure)))
   > Set Counter_Ready(Counter_Ready:double=0) -> (Output_Get:double)
   > n56 = Increment Float(Value=Get Counter() -> (Counter:double)) -> (Result=~)
   > Set Counter(Counter:double=@n56) -> (Output_Get:double)
   > n57 = GetCooldownTimeRemainingAndDuration!(execute, Actor:Actor Object Reference=Get PlayerPawn(), CooldownTags:Gameplay Tag Container Structure=Get Tag_CooldownActive()) -> (then, TimeRemaining:float, CooldownDuration:float)
   > HandleCooldown!(execute, self, Remaining Value:double=@n57, Duration:double=@n57.CooldownDuration) -> (then)
   > n58 = Branch(Condition="float > float"(A=Get Counter(), B=1))
   Stop Animations!(Start:Exec=@n58.else) -> (Finish:Exec) {is=title K2Node_MacroInstance}
   n59 = Get PlayerPawn()
   n60 = HasMatchingGameplayTag(self=Get PlayerPawn(), TagToCheck=(TagName="Granted.Athena.PlotTwist.Force.Power.Push"))
   n61 = Get Anim_Push() -> (Anim_Push:Widget Animation Object Reference)
   n62 = Get AnimThrow() -> (AnimThrow:Widget Animation Object Reference)
   > PlayAnimation(
    InAnimation=Select(
     Option 0:Widget Animation Object Reference=Select(Option 0=@n62, Option 1=@n61, Index=@n60)
     Option 1:Widget Animation Object Reference=Get Anim_Pull() -> (Anim_Pull:Widget Animation Object Reference)
     Index=HasMatchingGameplayTag(self=@n59, TagToCheck=(TagName="Granted.Athena.PlotTwist.Force.Power.Rip"))
    ) -> (ReturnValue:Widget Animation Object Reference)
    PlayMode=Reverse
    PlaybackSpeed=2
   )
   > PlayAnimationForward(InAnimation=Get Anim_CoolDownState() -> (Anim_CoolDownState:Widget Animation Object Reference))
   > n63 = GetDynamicMaterial(self=Get Reticle())
   > SetScalarParameterValue(self=@n63, ParameterName=TopPartFlash, Value=1)
   Set Counter(execute=@n55.else, Counter=0)
   > n64 = Increment Float(Value=Get Counter_Ready() -> (Counter_Ready:double)) -> (Result=~)
   > Set Counter_Ready(Counter_Ready=@n64)
   > n65 = Branch(Condition="float > float"(A=Get Counter_Ready(), B=1))
   n66 = Get AmberVineCooldownMaterialRef()
   SetScalarParameterValue(execute=@n65.else, self=@n66, ParameterName=isReady, Value=1)
   > SetScalarParameterValue(self=@n66, ParameterName=AnimatePipProgress, Value=1)
   > Stop Animations()
   > PlayAnimationReverse(InAnimation=Get Anim_CoolDownState())
   n67 = Get PlayerPawn()
   n68 = HasMatchingGameplayTag(self=Get PlayerPawn(), TagToCheck=(TagName="Granted.Athena.PlotTwist.Force.Power.Push"))
   n69 = Get Anim_Push()
   n70 = Get AnimThrow()
   > PlayAnimation(InAnimation=Select(
     Option 0=Select(Option 0=@n70, Option 1=@n69, Index=@n68)
     Option 1=Get Anim_Pull()
     Index=HasMatchingGameplayTag(self=@n67, TagToCheck=(TagName="Granted.Athena.PlotTwist.Force.Power.Rip"))
    ))
   > n71 = GetDynamicMaterial(self=Get Reticle())
   > SetScalarParameterValue(self=@n71, ParameterName=TopPartFlash, Value=0)
   n72 = Set AmmoQuantity(execute=[@n47.else, @n45.else, @n44.else], AmmoQuantity=-1) -> (Output_Get=0)
   n73 = SetVisibility(
    execute=@n42.else
    self=[
     Get Reticle()
     Get Reticle_Lightning_VFX() -> (Reticle_Lightning_VFX:Image Object Reference)
     Get Reticle_Pull() -> (Reticle_Pull:Image Object Reference)
     Get Reticle_Push() -> (Reticle_Push:Image Object Reference)
     Get Reticle_Throw() -> (Reticle_Throw:Image Object Reference)
     Get Reticle_Lightning()
    ]
    InVisibility=Hidden
   )
  ]}
 HandleCooldown {nodes [
   HandleCooldown!() -> (then, Remaining Value:double=~, Duration:double=~) {is=title K2Node_FunctionEntry}
   > SetScalarParameterValue(self=Get AmberVineCooldownMaterialRef(), ParameterName=isReady, Value=0)
   n1 = Get Remaining Value() -> (Remaining Value:double)
   n2 = Get Duration() -> (Duration:double)
   > n3 = Set PctComplete(PctComplete:double=SafeDivide(A=float - float(A=@n2, B=@n1), B=Get Duration())) -> (Output_Get:double)
   > SetScalarParameterValue(self=Get AmberVineCooldownMaterialRef(), ParameterName=AnimatePipProgress, Value=@n3)
   > Branch(Condition="float > float"(A=@n3, B=0.99))
   > PlayAnimation(InAnimation=Get Anim_ShowVFX() -> (Anim_ShowVFX:Widget Animation Object Reference))
  ]}
 ResetReticle {nodes [
   ResetReticle() {is=title K2Node_FunctionEntry}
   n1 = Get Homing() -> (Homing:Widget Animation Object Reference)
   > PlayAnimation(InAnimation=@n1, StartAtTime=GetEndTime(self=@n1), PlayMode=Reverse)
   > Set ShowingHoming(ShowingHoming:bool=false) -> (Output_Get:bool)
  ]}
 SetHomingState {nodes [
   n1 = SetHomingState!() -> (then, bHoming:bool=~) {is=title K2Node_FunctionEntry}
   > n2 = Branch(Condition=@n1)
   > Branch(Condition=NOT Boolean(A=Get ShowingHoming() -> (ShowingHoming:bool)))
   > PlayAnimation(InAnimation=Get Homing())
   > Set ShowingHoming(ShowingHoming=true)
   ResetReticle(execute=@n2.else) {is=title K2Node_CallFunction}
  ]}
 SetInRangeState {nodes [
   n1 = SetInRangeState!() -> (then, bInRange:bool=~) {is=title K2Node_FunctionEntry}
   > SetRenderOpacity(self=Get ReticleOverlay() -> (ReticleOverlay:Overlay Object Reference), InOpacity=SelectFloat(A=0.7, B=0.3, bPickA=@n1))
   SetVisibility(
    self=Get Center() -> (Center:Image Object Reference)
    InVisibility=Select(Option 0:ESlateVisibility Enum=Collapsed, Option 1:ESlateVisibility Enum=HitTestInvisible, Index=@n1) -> (ReturnValue:ESlateVisibility Enum)
   )
  ]}
 SetReticleCooldown {nodes [
   SetReticleCooldown() {is=title K2Node_FunctionEntry}
   n1 = Get PlayerPawn()
   > SetVisibility(
    self=Get Reticle_Push()
    InVisibility=Select(
     Option 0=Collapsed
     Option 1=Visible
     Index=HasMatchingGameplayTag(self=@n1, TagToCheck=(TagName="Granted.Athena.PlotTwist.Force.Power.Push"))
    )
   )
   n2 = Get PlayerPawn()
   > SetVisibility(
    self=Get Reticle_Pull()
    InVisibility=Select(
     Option 0=Collapsed
     Option 1=Visible
     Index=HasMatchingGameplayTag(self=@n2, TagToCheck=(TagName="Granted.Athena.PlotTwist.Force.Power.Rip"))
    )
   )
   n3 = Get PlayerPawn()
   > SetVisibility(
    self=Get Reticle_Throw()
    InVisibility=Select(
     Option 0=Collapsed
     Option 1=Visible
     Index=HasMatchingGameplayTag(self=@n3, TagToCheck=(TagName="Granted.Athena.PlotTwist.Force.Power.Projectile"))
    )
   )
   n4 = Get PlayerPawn()
   n5 = Get Tag_PassiveGranted_AmberVine()
   > SetVisibility(
    self=[Get Reticle_Lightning_VFX(), Get Reticle_Lightning()]
    InVisibility=Select(Option 0=Collapsed, Option 1=Visible, Index=HasMatchingGameplayTag(self=@n4, TagToCheck=@n5))
   )
   > n6 = GetDynamicMaterial(self=Get Reticle())
   n7 = HasMatchingGameplayTag(self=Get PlayerPawn(), TagToCheck=Get Tag_PassiveGranted_AmberVine())
   n8 = HasMatchingGameplayTag(self=Get PlayerPawn(), TagToCheck=Get Tag_Ability_HeavyLamp())
   > SetScalarParameterValue(self=@n6, ParameterName=isForceLightning, Value=SelectFloat(A=1, B=0, bPickA=OR Boolean(A=@n7, B=@n8)))
  ]}
 StopAnimations {nodes [
   Inputs!() -> (Start:Exec)
   > StopAnimation(InAnimation=Get Anim_Push())
   > StopAnimation(InAnimation=Get AnimThrow())
   > StopAnimation(InAnimation=Get Anim_Pull())
   > StopAnimation(InAnimation=Get Anim_CoolDownState())
   > StopAnimation(InAnimation=Get Anim_LockOn())
   > Outputs!(Finish:Exec)
  ]}
 UpdateReticleState {nodes [
   n1 = UpdateReticleState!() -> (then, bLocked:bool=~) {is=title K2Node_FunctionEntry}
   > SetColorAndOpacity(
    self:Image Object Reference=Get Reticle()
    InColorAndOpacity:Linear Color Structure=SelectColor(A=(R=1,G=0,B=0,A=1), B=(R=1,G=1,B=1,A=1), bPickA=@n1)
   )
  ]}
]
variables [
 AmberVine Component {type=BP_AmbervinePawnComponent_C, default=null}
 AmberVineCooldownMaterialRef {type=MaterialInstanceDynamic}
 Ammo_AmberVine {type=FortItemDefinition, default=/PlotTwist/Items/ForcePowers/AthenaAmmoData_AmberVine}
 Ammo_Projectile {type=FortItemDefinition, default=/PlotTwist/Items/ForcePowers/AthenaAmmoData_ForceProjectile}
 Ammo_Push {type=FortItemDefinition, default=/PlotTwist/Items/ForcePowers/AthenaAmmoData_ForcePush}
 Ammo_Rip {type=FortItemDefinition, default=/PlotTwist/Items/ForcePowers/AthenaAmmoData_ForceRip}
 AmmoHeavyLamp {type=FortItemDefinition, default=/PlotTwist/Items/ForcePowers/AthenaAmmoData_HeavyLamp}
 AmmoItemDefinition {type=FortItemDefinition, default=null}
 AmmoQuantity {type=integer}
 Cached Has Homing Target {type=boolean}
 Charges {type="array<Image>", default []}
 Counter {type=number}
 Counter_Ready {type=number}
 CounterPlayLockOnBackward {type=number}
 CounterPlayLockOnForward {type=number}
 PlayerController {type=FortPlayerController, default=null}
 PlayerPawn {type=FortPlayerPawnAthena}
 ShowingHoming {type=boolean}
 ShowReticle {type=boolean}
 Tag_Ability_HeavyLamp {type=GameplayTag, default {tagName=Granted.Athena.PlotTwist.Force.Power.HeavyLamp}}
 Tag_CooldownActive {type=GameplayTagContainer, default {gameplayTags [{tagName=Granted.Cooldown.Galileo.Lobster.ActivePower}]}}
 Tag_PassiveGranted_AmberVine {type=GameplayTag, default {tagName=Granted.Athena.PlotTwist.Force.Passive.AmberVine}}
 Tag_PassiveGranted_HeavyLamp {type=GameplayTag, default {tagName=Granted.Athena.PlotTwist.Force.Passive.HeavyLamp}}
 Tag_PassiveGranted_Projectile {type=GameplayTag, default {tagName=Granted.Athena.PlotTwist.Force.Passive.Projectile}}
 Tag_PassiveGranted_Push {type=GameplayTag, default {tagName=Granted.Athena.PlotTwist.Force.Passive.Push}}
 Tag_PassiveGranted_Rip {type=GameplayTag, default {tagName=Granted.Athena.PlotTwist.Force.Passive.Rip}}
 Tag_ShowReticle {type=GameplayTagContainer, default {gameplayTags [{tagName=Granted.Athena.PlotTwist.Force.ShowReticle}]}}
 Target Location Is in Range {type=boolean}
 Task_WaitPassiveGranted_AmberVine {type=AbilityAsync_WaitGameplayTagAdded, default=null}
 Task_WaitPassiveGranted_HeavyLamp {type=AbilityAsync_WaitGameplayTagAdded, default=null}
 Task_WaitPassiveGranted_Projectile {type=AbilityAsync_WaitGameplayTagAdded, default=null}
 Task_WaitPassiveGranted_Push {type=AbilityAsync_WaitGameplayTagAdded, default=null}
 Task_WaitPassiveGranted_Rip {type=AbilityAsync_WaitGameplayTagAdded, default=null}
 Temp_ShowChargesEnabled {type=boolean, default=true}
]
widgets {
 info {
  inheritedWidgetCount=1
  parentClass=/Script/FortniteUI.FortWeaponReticleExtensionWidgetBase
  rootWidgetClass=&Overlay
  widgetCount=9
 }
 widgets [
  Overlay_0 {widgetClassPath=&Overlay, widget=Overlay_0}
  ReticleOverlay {
   widgetClassPath=&Overlay
   bIsVariable=true
   parent=Overlay_0
   widget=ReticleOverlay
   properties {renderOpacity=0.4}
   slot {class=OverlaySlot}
  }
  Reticle {
   widgetClassPath=&Image
   bIsVariable=true
   parent=ReticleOverlay
   widget=Reticle
   properties {
    visibility=HitTestInvisible
    brush {
     bIsDynamicallyLoaded=false
     drawAs=Image
     imageType=NoImage
     mirroring=NoMirror
     resourceName=None
     resourceObject=/PlotTwist/ForcePowers/Shared/Reticle/Materials/MI_UI_ForcePowersReticle
     tiling=NoTile
     imageSize=<84 84>
     margin=<ltrb 0 0 0 0>
     outlineSettings {
      bUseBrushTransparency=false
      roundingType=HalfHeightRadius
      width=0
      color {colorUseRule=UseColor_Specified, specifiedColor=<rgba 0 0 0 0>}
      cornerRadii=<0 0 0 0>
     }
     tintColor {colorUseRule=UseColor_Specified, specifiedColor=<rgba 1 1 1 1>}
     uVRegion {bIsValid=false, max=<0 0>, min=<0 0>}
    }
   }
   slot {class=OverlaySlot, properties {horizontalAlignment=HAlign_Fill, verticalAlignment=VAlign_Fill}}
  }
  Reticle_Push {
   widgetClassPath=&Image
   bIsVariable=true
   parent=ReticleOverlay
   widget=Reticle_Push
   properties {
    visibility=HitTestInvisible
    brush {
     bIsDynamicallyLoaded=false
     drawAs=Image
     imageType=NoImage
     mirroring=NoMirror
     resourceName=None
     resourceObject=/PlotTwist/ForcePowers/Shared/Reticle/Materials/MI_UI_CoolDownStatePush
     tiling=NoTile
     imageSize=<28 28>
     margin=<ltrb 0 0 0 0>
     outlineSettings {
      bUseBrushTransparency=false
      roundingType=HalfHeightRadius
      width=0
      color {colorUseRule=UseColor_Specified, specifiedColor=<rgba 0 0 0 0>}
      cornerRadii=<0 0 0 0>
     }
     tintColor {colorUseRule=UseColor_Specified, specifiedColor=<rgba 1 1 1 1>}
     uVRegion {bIsValid=false, max=<0 0>, min=<0 0>}
    }
   }
   slot {class=OverlaySlot, properties {horizontalAlignment=HAlign_Center, verticalAlignment=VAlign_Center, padding=<ltrb 0 29 0 0>}}
  }
  Reticle_Throw {
   widgetClassPath=&Image
   bIsVariable=true
   parent=ReticleOverlay
   widget=Reticle_Throw
   properties {
    visibility=HitTestInvisible
    brush {
     bIsDynamicallyLoaded=false
     drawAs=Image
     imageType=NoImage
     mirroring=NoMirror
     resourceName=None
     resourceObject=/PlotTwist/ForcePowers/Shared/Reticle/Materials/M_UI_CoolDownStateThrow
     tiling=NoTile
     imageSize=<84 84>
     margin=<ltrb 0 0 0 0>
     outlineSettings {
      bUseBrushTransparency=false
      roundingType=HalfHeightRadius
      width=0
      color {colorUseRule=UseColor_Specified, specifiedColor=<rgba 0 0 0 0>}
      cornerRadii=<0 0 0 0>
     }
     tintColor {colorUseRule=UseColor_Specified, specifiedColor=<rgba 1 1 1 1>}
     uVRegion {bIsValid=false, max=<0 0>, min=<0 0>}
    }
   }
   slot {class=OverlaySlot, properties {horizontalAlignment=HAlign_Center, verticalAlignment=VAlign_Center}}
  }
  Reticle_Pull {
   widgetClassPath=&Image
   bIsVariable=true
   parent=ReticleOverlay
   widget=Reticle_Pull
   properties {
    visibility=HitTestInvisible
    brush {
     bIsDynamicallyLoaded=false
     drawAs=Image
     imageType=NoImage
     mirroring=NoMirror
     resourceName=None
     resourceObject=/PlotTwist/ForcePowers/Shared/Reticle/Materials/M_UI_CoolDownStatePull
     tiling=NoTile
     imageSize=<60 60>
     margin=<ltrb 0 0 0 0>
     outlineSettings {
      bUseBrushTransparency=false
      roundingType=HalfHeightRadius
      width=0
      color {colorUseRule=UseColor_Specified, specifiedColor=<rgba 0 0 0 0>}
      cornerRadii=<0 0 0 0>
     }
     tintColor {colorUseRule=UseColor_Specified, specifiedColor=<rgba 1 1 1 1>}
     uVRegion {bIsValid=false, max=<0 0>, min=<0 0>}
    }
    colorAndOpacity=<rgba 1 1 1 0.6>
    renderTransform {angle=0, scale=<1 1>, shear=<0 0>, translation=<0 4>}
   }
   slot {class=OverlaySlot, properties {horizontalAlignment=HAlign_Center, verticalAlignment=VAlign_Center, padding=<ltrb 0 13 0 0>}}
  }
  Reticle_Lightning {
   widgetClassPath=&Image
   bIsVariable=true
   parent=Overlay_0
   widget=Reticle_Lightning
   properties {
    visibility=HitTestInvisible
    brush {
     bIsDynamicallyLoaded=false
     drawAs=Image
     imageType=FullColor
     mirroring=NoMirror
     resourceName=None
     resourceObject=/PlotTwist/ForcePowers/Shared/Reticle/Materials/MI_UI_ForceLightning_Cooldown
     tiling=NoTile
     imageSize=<40 40>
     margin=<ltrb 0 0 0 0>
     outlineSettings {
      bUseBrushTransparency=false
      roundingType=HalfHeightRadius
      width=0
      color {colorUseRule=UseColor_Specified, specifiedColor=<rgba 0 0 0 0>}
      cornerRadii=<0 0 0 0>
     }
     tintColor {colorUseRule=UseColor_Specified, specifiedColor=<rgba 1 1 1 1>}
     uVRegion {bIsValid=false, max=<0 0>, min=<0 0>}
    }
    renderTransform {angle=0, scale=<1 1>, shear=<0 0>, translation=<64 0>}
   }
   slot {class=OverlaySlot, properties {horizontalAlignment=HAlign_Center, verticalAlignment=VAlign_Center}}
  }
  Reticle_Lightning_VFX {
   widgetClassPath=&Image
   bIsVariable=true
   parent=Overlay_0
   widget=Reticle_Lightning_VFX
   properties {
    renderOpacity=0
    visibility=HitTestInvisible
    brush {
     bIsDynamicallyLoaded=false
     drawAs=Image
     imageType=FullColor
     mirroring=NoMirror
     resourceName=None
     resourceObject=/PlotTwist/ForcePowers/Shared/Reticle/Materials/MI_UI_ForcePowers_Cooldown_Lightning
     tiling=NoTile
     imageSize=<64 64>
     margin=<ltrb 0 0 0 0>
     outlineSettings {
      bUseBrushTransparency=false
      roundingType=HalfHeightRadius
      width=0
      color {colorUseRule=UseColor_Specified, specifiedColor=<rgba 0 0 0 0>}
      cornerRadii=<0 0 0 0>
     }
     tintColor {colorUseRule=UseColor_Specified, specifiedColor=<rgba 1 1 1 1>}
     uVRegion {bIsValid=false, max=<0 0>, min=<0 0>}
    }
    renderTransform {angle=0, scale=<1 1>, shear=<0 0>, translation=<64 0>}
   }
   slot {class=OverlaySlot, properties {horizontalAlignment=HAlign_Center, verticalAlignment=VAlign_Center}}
  }
  TutorialHighlight {widgetClassPath=&TutorialHighlight}
 ]
}
